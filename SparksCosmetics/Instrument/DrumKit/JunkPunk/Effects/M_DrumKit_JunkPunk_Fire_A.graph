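# Translucent, unlit, two-sided flame material for Junk Punk drum kit effects, built for mesh particles and skeletal meshes.
# Three panned noise layers, shaped by a gradient or Wipe-driven swipe mask, blend inner and outer colors into HDR emissive.
# Opacity combines that mask with Fresnel, depth fade, particle alpha and optional edge masking.
# World position offset adds noise along normals, sine waves and an optional clamped pull toward a dynamic-parameter position.
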Material /SparksCosmetics/Instrument/DrumKit/JunkPunk/Effects/M_DrumKit_JunkPunk_Fire_A
graph {
 properties {
  bUsedWithMeshDeformer=true
  bUsedWithNanite=true
  bUsedWithNiagaraMeshParticles=true
  bUsedWithSkeletalMesh=true
  blendMode=BLEND_Translucent
  pixelDepthOffsetMode=PDOM_Legacy
  previewMesh=&/SM_JunkPunk_Flame_B
  shadingModel=MSM_Unlit
  twoSided=true
  usedShadingModels=Unlit
 }
 nodes [
  n1 = ComponentMask(None=TextureCoordinate()) {g=true}
  n2 = MaterialFunctionCall(In=Power(Base=OneMinus(None=@n1)) {constExponent=1}, Contrast=Constant() {r=0.5}) {materialFunction=&CheapContrast}
  n3 = Add(A=TextureCoordinate() {uTiling=2}, B=AppendVector(A=ParticleRandom(), B=Constant())) {constB=0}
  n4 = Time()
  n5 = Sine(None=Add(A=ComponentMask(None=@n3) {r=true}, B=Multiply(A=@n4) {constB=-0.5})) {period=1.3}
  n6 = Transform(None=Multiply(A=@n5, B=VectorParameter() {parameterName=WPO Vector, defaultValue=<rgba 0 1 0 1>}.RGB)) {transformSourceType=TRANSFORMSOURCE_Local}
  n7 = ScalarParameter() {defaultValue=1, parameterName=DistortHorizontal}
  n8 = VectorParameter() {parameterName=WPO Vector2, defaultValue=<rgba 1 1 0 1>}
  n9 = Multiply(A=Transform(None=Multiply(A=@n5, B=@n8.RGB)) {transformSourceType=TRANSFORMSOURCE_Local})
  n10 = Multiply(A=ComponentMask(None=TextureCoordinate()) {r=true}) {constB=3}
  n11 = Multiply(A=Add(A=Multiply(A=@n6, B=@n7), B=Multiply(A=@n7, B=@n9)), B=LinearInterpolate(Alpha=Saturate(None=@n10)))
  n12 = Multiply(A=Saturate(None=Multiply(A=@n2.Result)), B=@n11)
  MaterialFunctionCall(X=@n12, Y=@n12, Z=Constant()) {materialFunction=&MakeFloat3}
  MaterialFunctionCall() {materialFunction=&DebugTimeSine}
  n13 = TextureCoordinate()
  n14 = MaterialFunctionCall!(Width, Sharpness, Offset, Direction Switch=StaticBool() {value=true}, Compare, Input Coordinates=Panner(Coordinate=@n13) {speedX=-1}) {materialFunction=/Engine/Functions/Engine_MaterialFunctions03/Procedurals/GeneratedBand}
  Multiply(A=@n14.Result)
  Multiply(A=@n12, B=ConstantBiasScale(None=@n14.Result) {bias=-0.5})
  n15 = VertexColor()
  Multiply(A=Power(Base=@n15.R) {constExponent=10}) {constB=5}
  n16 = Fresnel(ExponentIn=Constant() {r=5}) {baseReflectFraction=0.01}
  OneMinus(None=@n16)
  n17 = ComponentMask(None=@n13) {g=true}
  n18 = VertexNormalWS()
  n19 = ScalarParameter() {defaultValue=1, parameterName=ThirdTextureXScale}
  n20 = ScalarParameter() {defaultValue=0.5, parameterName=ThirdTextureYScale}
  n21 = ComponentMask(None=VectorParameter() {group=UVScale, parameterName=UVScale, defaultValue=<rgba 1 1 0 0>}.RGB) {g=true, r=true}
  n22 = Multiply(A=Multiply(A=TextureCoordinate(), B=AppendVector(A=@n19, B=@n20)), B=@n21)
  n23 = DynamicParameter() {paramNames [UVOffset,Wipe,Param3,Speed]}
  n24 = Add(A=@n15.R, B=@n23.UVOffset) {constB=0}
  n25 = TextureSample(UVs=Panner(Coordinate=Add(A=@n22, B=@n24)) {speedX=-0.05, speedY=1}) {texture=&T_NoisePack_01}
  n26 = Constant() {r=0.2}
  n27 = Multiply(
   A=TextureCoordinate()
   B=AppendVector(
    A=ScalarParameter() {defaultValue=4, parameterName=SecondTextureXScale}
    B=ScalarParameter() {defaultValue=1, parameterName=SecondTextureYScale}
   )
  )
  n28 = Multiply(
   A=LinearInterpolate(A=ScalarParameter() {defaultValue=4, parameterName=Speed}, B=@n23.Speed, Alpha=ScalarParameter() {parameterName=Use Dynamic Speed})
   B=Time()
  )
  n29 = Panner(Coordinate=Add(A=Multiply(A=@n21, B=@n27), B=@n24), Time=@n28) {speedX=0.1, speedY=1}
  n30 = TextureCoordinate()
  n31 = AppendVector(
   A=ScalarParameter() {defaultValue=12, parameterName=FirstTextureXScale}
   B=ScalarParameter() {defaultValue=3, parameterName=FirstTextureYScale}
  )
  n32 = Add(A=@n24, B=Multiply(A=@n21, B=Multiply(A=@n30, B=@n31)))
  n33 = Multiply(
   A=Add(
    A=MaterialFunctionCall(In=@n25.G, Contrast=@n26) {materialFunction=&CheapContrast}.Result
    B=TextureSample(UVs=@n29) {texture=&T_NoisePack_01}.G
   )
   B=TextureSample(UVs=Panner(Coordinate=@n32, Time=@n28) {speedX=-0.1, speedY=1}) {texture=&T_NoisePack_01}.G
  )
  n34 = ScalarParameter() {defaultValue=5, parameterName=Distort}
  n35 = Panner(Coordinate=Add(A=@n24, B=@n13), Time=@n28) {speedY=0.25}
  n36 = Add(
   A=Multiply(A=Multiply(A=@n18, B=@n33), B=@n34)
   B=Multiply(A=TextureSample(UVs=@n35) {texture=/Game/Global/Textures/Default/Masks_and_Gradients/CylGradient}.RGB, B=@n34) {constB=35}
  )
  n37 = Multiply(A=@n36, B=Power(Base=OneMinus(None=@n17), Exp=ScalarParameter() {defaultValue=1, group=WPO, parameterName=WPO Mask Exponent}))
  n38 = ScalarParameter() {group=WPO, parameterName=Toggle WPO Mask}
  n39 = DynamicParameter() {parameterIndex=1, defaultValue=<rgba 1 1 1 0>, paramNames [X,Y,Z,WPOIntensity]}
  n40 = Subtract(
   A=MaterialFunctionCall(X=@n39.X, Y=@n39.X, Z=@n39.X) {materialFunction=&MakeFloat3}.Result
   B=ParticlePositionWS()
  )
  n41 = Constant() {r=-30}
  n42 = Constant() {r=30}
  n43 = Power(Base=ComponentMask(None=TextureCoordinate()) {r=true}, Exp=ScalarParameter() {defaultValue=2, parameterName=WPOBlendPower})
  n44 = LinearInterpolate(
   B=LinearInterpolate(B=Clamp(None=@n40, Min=@n41, Max=@n42), Alpha=Multiply(A=@n39.WPOIntensity, B=@n43))
   Alpha=ScalarParameter() {parameterName=Use Interpolate WPO}
  )
  Reroute(None=Add(A=LinearInterpolate(A=@n36, B=@n37, Alpha=@n38), B=Add(A=@n44, B=@n11) {constB=0})) {desc=WPO}
  n45 = DynamicParameter() {desc=Color Outer, parameterIndex=1, paramNames [R,G,B,A]}
  n46 = ScalarParameter() {group=Color, parameterName=Use Dynamic Parameter Colors, sliderMax=1}
  n47 = DynamicParameter() {desc=Color Inner, parameterIndex=2, paramNames [R,G,B,A]}
  n48 = TextureCoordinate()
  n49 = ComponentMask(None=@n48) {r=true}
  n50 = Multiply(A=@n49, B=OneMinus(None=@n49))
  n51 = MaterialFunctionCall(Float2=TextureCoordinate()) {materialFunction=&BreakOutFloat2Components}
  n52 = DynamicParameter() {paramNames [UVOffset,Wipe,Param3,Speed]}
  n53 = AppendVector(A=@n51.R, B=Add(A=@n51.G, B=LinearInterpolate(Alpha=@n52.Wipe) {constA=-1}))
  n54 = Multiply(A=Power(Base=TextureSample(UVs=@n53) {samplerSource=SSM_Clamp_WorldGroupSettings, texture=&CylGradient}.G)) {constB=20}
  n55 = Multiply(
   A=Power(Base=ComponentMask(None=@n48) {g=true}, Exp=ScalarParameter() {defaultValue=2, parameterName=SwipeMaskPower})
   B=Multiply(A=Multiply(A=@n50) {constB=4}, B=Clamp(None=@n54))
  )
  n56 = Multiply(A=@n55, B=@n33)
  n57 = MaterialFunctionCall() {materialFunction=&LinearGradient}
  n58 = ScalarParameter() {defaultValue=15, parameterName=FireLine}
  n59 = Power(Base=Add(A=@n33, B=Power(Base=@n57.VGradient, Exp=@n58))) {constExponent=0.5}
  n60 = Clamp(None=MaterialFunctionCall(
    In=StaticSwitchParameter(True=@n56, False=@n59) {parameterName=UseSwipe}
    Contrast=ScalarParameter() {defaultValue=0.5, parameterName=HardEdgeness}
   ) {materialFunction=&CheapContrast}.Result)
  n61 = LinearInterpolate(
   A=LinearInterpolate(
    A=VectorParameter() {desc=Color Outer, group=Color, parameterName=Color_A, defaultValue=<rgba 10 0 0.1768 1>}.RGB
    B=MaterialFunctionCall(X=@n45.R, Y=@n45.R, Z=@n45.R) {materialFunction=&MakeFloat3}.Result
    Alpha=@n46
   )
   B=LinearInterpolate(
    A=VectorParameter() {desc=Color Inner, group=Color, parameterName=Color_B, defaultValue=<rgba 10 1.3123 0 1>}.RGB
    B=MaterialFunctionCall(X=@n47.R, Y=@n47.R, Z=@n47.R) {materialFunction=&MakeFloat3}.Result
    Alpha=@n46
   )
   Alpha=@n60
  )
  n62 = Clamp(None=@n61)
  n63 = ParticleColor()
  n64 = ScalarParameter() {defaultValue=1, parameterName=Use Particle Color}
  n65 = Multiply(A=@n61, B=LinearInterpolate(B=@n63.RGB, Alpha=@n64) {constA=1})
  Reroute(None=MaterialFunctionCall(Emissive=MaterialFunctionCall(LDR Input=@n62, HDR Tint=@n65) {materialFunction=&DeriveHDRfromLDR}.RemappedRange) {materialFunction=&MF_FX_EmissiveAdjustments}.Result)
  n66 = Multiply(A=Power(Base=LinearInterpolate(Alpha=@n16) {constA=1, constB=0}) {constExponent=12})
  n67 = LinearInterpolate(B=@n63.A, Alpha=@n64) {constA=1}
  n68 = ComponentMask(None=TextureCoordinate()) {r=true}
  n69 = OneMinus(None=@n68)
  n70 = LinearInterpolate(B=Multiply(A=Multiply(A=@n69, B=@n68)) {constB=4}, Alpha=ScalarParameter() {parameterName=Use Edge Mask}) {constA=1}
  n71 = Power(
   Base=MaterialFunctionCall() {materialFunction=&LinearGradient}.VGradient
   Exp=ScalarParameter() {defaultValue=8, parameterName=GradientContrast}
  )
  n72 = MaterialFunctionCall(Axis=Constant3Vector() {constant=<rgba 0 0 1 0>}, Exponent=ScalarParameter() {defaultValue=3, parameterName=FresnelExp}) {materialFunction=&AxisAlignedFresnel}
  n73 = Constant() {r=1}
  n74 = Multiply(
   A=@n60
   B=Multiply(
    A=StaticSwitchParameter(True=@n55, False=@n71) {parameterName=UseSwipe}
    B=StaticSwitchParameter(True=@n72.Fresnel, False=@n73) {defaultValue=true, parameterName=UseAxisAlignFresnel}
   )
  )
  n75 = ScalarParameter() {defaultValue=64, parameterName=FadeDistance}
  n76 = Saturate(None=Multiply(A=Multiply(A=@n67, B=@n70), B=DepthFade(Opacity=@n74, FadeDistance=@n75) {fadeDistanceDefault=50}))
  n77 = ScalarParameter() {group=Fresnel, parameterName=Fresnel Brightness, sortPriority=1}
  n78 = ScalarParameter() {group=Fresnel, parameterName=Frensel Exponent, sortPriority=7}
  n79 = LinearInterpolate(
   A=@n76
   B=Multiply(
    A=@n76
    B=MaterialFunctionCall(Brightness=@n77, Exponent=@n78) {materialFunction=&MF_Fresnel_FX}.Fresnel
   )
   Alpha=ScalarParameter() {group=Fresnel, parameterName=Use Fresnel, sortPriority=0}
  )
  Reroute(None=Multiply(A=Saturate(None=@n79), B=Saturate(None=@n66))) {desc=OPACITY}
  ScalarParameter() {defaultValue=1, parameterName=DistortHorizontal}
 ]
}
